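# Material function for the ScarB car body's Ramuh skin, setting base color, metallic and roughness.
# UV1-driven front and rear gradients blend colors with animated Perlin-modulated dot noise, then an RGBA mask texture layers roof fade, fixed dark areas and text color.
# Two mirrored logo samples are added, with both sides using LogoTransformL's scale.
# The paint replaces body color through the ScarB mask, then a view-angle falloff shades the whole result.
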MaterialFunction /VehicleCosmetics/Bodies/ScarB/Skins/Ramuh/MF_Skin_ScarBRamuh
graph {nodes [
  n1 = SetMaterialAttributes!(
   MaterialAttributes
   BaseColor=NamedRerouteUsage()
   Metallic=ScalarParameter() {parameterName=Metallic}
   Roughness=ScalarParameter() {defaultValue=0.3, parameterName=Roughness}
   EmissiveColor
   Normal
   AmbientOcclusion
  ) {attributeSetTypes [
    #1
    #2
    #3
    #4
    #5
    #6
   ]}
  FunctionOutput(None=@n1)
  n2 = TextureCoordinate() {coordinateIndex=1}
  n3 = ComponentMask(None=@n2) {g=true}
  n4 = ComponentMask(None=@n2) {r=true}
  n5 = AppendVector(A=@n3, B=OneMinus(None=@n4))
  n6 = VectorParameter() {desc="(R,G) = Detail1    (B,A) = Detail2", parameterName=LogoTransformL, defaultValue=<rgba 10 10 -4 -6.6>}
  n7 = ComponentMask(None=@n6.RGB) {g=true, r=true}
  NamedRerouteDeclaration(None=Add(A=Multiply(A=@n5, B=@n7), B=ComponentMask(None=@n6.RGBA) {a=true, b=true})) {name=Logo_L_UVs, nodeColor=<rgba 0.1484 0 1 1>}
  NamedRerouteDeclaration(None=Time() {bOverride_Period=true, period=500}) {name=time, nodeColor=<rgba 1 0 0.4219 1>}
  n8 = NamedRerouteUsage()
  n9 = NamedRerouteUsage()
  n10 = ScalarParameter() {defaultValue=0.1, parameterName=FrontGradient_Shift}
  n11 = Saturate(None=Multiply(A=Subtract(A=@n9, B=@n10), B=ScalarParameter() {defaultValue=4, parameterName=FrontGradient_Sharpen}))
  n12 = LinearInterpolate(A=NamedRerouteUsage(), B=@n8, Alpha=@n11)
  n13 = NamedRerouteUsage()
  n14 = MaterialFunctionCall!(Texture Coordinates, Tiling X=Constant() {r=2}, Tiling Y=Constant() {r=0.05}, PanSpeedX, PanSpeedY, Angle, RotateSpeed, Balance, Chaos) {materialFunction=/VehicleCosmetics/Materials/MaterialFunctions/MF_Noisemaker_Perlin}
  n15 = Subtract(A=Multiply(A=TextureCoordinate() {coordinateIndex=1}) {constB=0.75}, B=AppendVector(A=Constant(), B=Multiply(A=@n14.Perlin2) {constB=0.05}))
  n16 = Constant() {r=0.5}
  n17 = Multiply(A=@n14.Perlin2) {constB=0.58}
  n18 = Add(A=@n14.Perlin3) {constB=0.3}
  n19 = LinearInterpolate(
   A=@n13
   B=LinearInterpolate(
    A=@n13
    B=NamedRerouteUsage()
    Alpha=MaterialFunctionCall!(Texture Coordinates=@n15, Tiling X, Tiling Y, Offset, Angle, Sharpen, PanSpeedX, PanSpeedY=@n16, RotateSpeed) {materialFunction=/VehicleCosmetics/Materials/MaterialFunctions/MF_Noisemaker_Dots}.Checkers
   )
   Alpha=Saturate(None=LinearInterpolate(A=@n17, B=@n18) {constAlpha=0.8})
  )
  n20 = NamedRerouteUsage()
  n21 = ScalarParameter() {defaultValue=0.5, parameterName=RearGradient_Shift}
  n22 = Saturate(None=Multiply(A=Subtract(A=@n9, B=@n21), B=ScalarParameter() {defaultValue=5, parameterName=RearGradient_Sharpen}))
  n23 = TextureSampleParameter2D() {parameterName=RGB, texture=&/T_ScarB_Ramuh_RGB}
  n24 = LinearInterpolate(
   A=LinearInterpolate(A=LinearInterpolate(A=@n12, B=@n19, Alpha=@n11), B=@n20, Alpha=@n22)
   B=LinearInterpolate(A=@n8, B=LinearInterpolate(A=@n8, B=@n20, Alpha=@n22), Alpha=ScalarParameter() {parameterName=FadeRoof})
   Alpha=@n23.G
  )
  n25 = Constant3Vector() {constant=<rgba 0.01 0.01 0.01 0>}
  n26 = LinearInterpolate(
   A=LinearInterpolate(A=LinearInterpolate(A=@n24, B=@n25, Alpha=@n23.A), B=NamedRerouteUsage(), Alpha=@n23.B)
   B=VectorParameter() {parameterName=TextColor, defaultValue=<rgba 0.82 0.82 0.82 1>}.RGB
   Alpha=@n23.R
  )
  n27 = TextureSampleParameter2D(UVs=NamedRerouteUsage()) {parameterName=Logo, texture=&/T_ScarB_Ramuh01_Logo}
  n28 = TextureSampleParameter2D(UVs=NamedRerouteUsage()) {parameterName=Logo, texture=&T_ScarB_Ramuh01_Logo}
  n29 = LinearInterpolate(
   A=TextureSample() {texture=/VehicleCosmetics/Bodies/ScarB/Textures/Body/T_ScarB_BC}.RGB
   B=LinearInterpolate(A=LinearInterpolate(A=@n26, B=@n27.RGB, Alpha=@n27.RGB), B=@n28.RGB, Alpha=@n28.RGB)
   Alpha=TextureSample() {texture=/VehicleCosmetics/Bodies/ScarB/Textures/Body/T_ScarB_M}.R
  )
  n30 = DotProduct(
   A=Transform(None=CameraVectorWS()) {transformSourceType=TRANSFORMSOURCE_World, transformType=TRANSFORM_Tangent}
   B=Constant3Vector() {constant=<rgba 0 0 1 0>}
  )
  n31 = Power(Base=OneMinus(None=@n30)) {constExponent=3}
  n32 = Power(Base=OneMinus(None=@n31), Exp=ScalarParameter() {defaultValue=10, group=Finish, parameterName=FinishInteriorFalloff})
  n33 = Power(Base=@n31, Exp=ScalarParameter() {defaultValue=1, group=Finish, parameterName=FinishOuterFalloff})
  n34 = Max(A=Saturate(None=Multiply(A=@n32) {constB=0.75}), B=Saturate(None=Add(A=@n33) {constB=0.2}))
  n35 = Power(Base=@n30) {constExponent=0.5}
  n36 = ConstantBiasScale(None=@n35) {bias=-0.95}
  NamedRerouteDeclaration(None=Add(A=Multiply(A=@n29, B=@n34), B=Saturate(None=@n36))) {name=BaseColor, nodeColor=<rgba 1 0.6328 0 1>}
  n37 = VectorParameter() {desc="Alpha = Emissive Color Brightness", group=Colors, parameterName=MainColor, defaultValue=<rgba 0.8148 0 0 1>}
  NamedRerouteDeclaration(None=@n37.RGB) {name=MainColor, nodeColor=<rgba 1 0 0 1>}
  NamedRerouteDeclaration(None=VectorParameter() {group=Colors, parameterName=SecondaryColor, defaultValue=<rgba 0.8148 0.8148 0.8148 1>}.RGB) {name=SecondaryColor, nodeColor=<rgba 1 1 1 1>}
  NamedRerouteDeclaration(None=VectorParameter() {group=Colors, parameterName=ThirdColor, defaultValue=<rgba 0.01 0.01 0.01 1>}.RGB) {name=ThirdColor, nodeColor=<rgba 0.01 0.01 0.01 1>}
  n38 = OneMinus(None=NamedRerouteDeclaration(None=@n3) {name=Gradient, nodeColor=<rgba 1 0 0.8906 1>})
  NamedRerouteDeclaration(None=@n38) {name=Gradient1-, nodeColor=<rgba 0 1 0.7891 1>}
  n39 = AppendVector(A=@n38, B=@n4)
  n40 = VectorParameter() {desc="(R,G) = Detail1    (B,A) = Detail2", parameterName=LogoTransformR, defaultValue=<rgba 10 10 -5 -6.6>}
  NamedRerouteDeclaration(None=Add(A=Multiply(A=@n39, B=@n7), B=ComponentMask(None=@n40.RGBA) {a=true, b=true})) {name=Logo_R_UVs, nodeColor=<rgba 0.2422 0 1 1>}
  OneMinus(None=@n35)
 ]}
